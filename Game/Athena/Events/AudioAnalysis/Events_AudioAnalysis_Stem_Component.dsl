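; Audio stem analysis component for Athena events that turns precomputed (NRT) audio analysis into a material parameter collection scalar.
; Update takes a playback time and, within the sound's duration, samples Loudness, Onset or ConstantQ by stem type; PostUpdate then pushes the stored scalar.
; Loudness feeds the scalar directly, while ConstantQ only stores its result.
; Onset flags strong recent hits with a cooldown check but sets the scalar to one regardless.
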(event EventBeginPlay)

(event EventTick
  (params (DeltaSeconds :type Float%20%28single-precision%29)))

(fn PostUpdate
  (bind _materialpc (Variables|Data|GetMaterialPC)
    (:then
      (Rendering|Material|SetScalarParameterValue _materialpc (Variables|Data|GetParameterName) (Variables|Private|GetScalarValue) :!class /Script/Engine.KismetMaterialLibrary))
    (:else)))

(fn Update
  (params (Time :type Float%20%28double-precision%29))
  (bind _timelocal (LocalVariables|GetTimeLocal))
  (LocalVariables|SetTimeLocal Time)
  (Utilities|FlowControl|Sequence
    (:then_0
      (if (not (> (LocalVariables|GetTimeLocal) (Variables|/Script/Engine.SoundBase|GetDuration (Variables|/Script/AudioAnalyzer.AudioAnalyzerNRT|GetSound (Variables|Data|GetNRT)))))
        (switch Utilities|FlowControl|Switch|SwitchonEvents_AudioAnalysis_Stem_Type (Variables|Data|GetType)
          (:NewEnumerator0
            (CallFunction|UpdateLoudness _timelocal))
          (:NewEnumerator1
            (CallFunction|UpdateOnset _timelocal))
          (:NewEnumerator2
            (CallFunction|UpdateConstantQ _timelocal)))))
    (:then_1
      (CallFunction|PostUpdate))))

(fn UpdateConstantQ
  (params (InSeconds :type Float%20%28double-precision%29))
  (bind _asconstant_qnrt (Utilities|Casting|CastToConstantQNRT (Variables|Data|GetNRT)))
  (Variables|Private|SetNRTConstantQ (AudioAnalyzer|GetNormalizedChannelConstantQAtTime _asconstant_qnrt InSeconds 0)))

(fn UpdateLoudness
  (params (InSeconds :type Float%20%28double-precision%29))
  (bind _asloudness_nrt (Utilities|Casting|CastToLoudnessNRT (Variables|Data|GetNRT)))
  (bind _output_get (Variables|Private|SetNRTLoudness (AudioAnalyzer|GetNormalizedLoudnessatTime _asloudness_nrt InSeconds)))
  (Variables|Private|SetScalarValue _output_get))

(fn UpdateOnset
  (params (InputPin :type Float%20%28double-precision%29))
  (bind _asonset_nrt (Utilities|Casting|CastToOnsetNRT (Variables|Data|GetNRT)))
  (Variables|Private|SetNRTOnsetDetected (and (> (.MaxValue (Math|Float|MaxOfFloatArray (.OutOnsetStrengths (AudioAnalyzer|GetNormalizedChannelOnsetsBetweenTimes _asonset_nrt (- InputPin 0.1) InputPin 0)))) 0.5) (> (Math|Float|Absolute(Float) (- (Utilities|Time|GetGameTimeinSeconds) (Variables|Private|GetTimeLastOnsetDetected))) 0.25)))
  (Variables|Private|SetScalarValue 1.0))
